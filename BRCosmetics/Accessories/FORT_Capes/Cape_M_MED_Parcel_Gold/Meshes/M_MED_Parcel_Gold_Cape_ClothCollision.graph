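# Physics asset that, by its name, supplies cloth collision for the M_MED Parcel Gold cape cosmetic.
# A single body on the backpack_root bone holds two overlapping capsules, rotated 90 degrees and offset from the bone origin, forming a simple collision volume.
# It defines no constraints, so it describes collision shapes only, with no jointed ragdoll setup.
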PhysicsAsset /BRCosmetics/Accessories/FORT_Capes/Cape_M_MED_Parcel_Gold/Meshes/M_MED_Parcel_Gold_Cape_ClothCollision
bodies {backpack_root [
  {shapeType=Capsule, length=1.3061, radius=8.2193, center=<0 -14.1076 -11.4022>, rotation=<rot 90 0 0>}
  {shapeType=Capsule, length=8.1562, radius=11.3438, shapeName=backpack_root_capsule, center=<0 -14.7615 -5.7221>, rotation=<rot 90 0 0>}
 ]}
constraints []
